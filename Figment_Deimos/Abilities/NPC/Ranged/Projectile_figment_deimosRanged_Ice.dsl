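; Ice projectile fired by the Figment Deimos ranged NPC, exploding on impact or when it stops on the server.
; BeginPlay shows a gunshot sound indicator at the instigator for hostile and neutral affiliations.
; Hit and stop share an Exploded guard, then spawn break particles, disable collision and Niagara, and call DoExplosion.
; Explosion makes AI noise, plays an impact cue and fades the projectile sound, which destruction also fades if playing.
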(event EventBeginPlay
  (|CallParentFunction)
  (Fortnite|ShowSoundIndicatorLocalforTeam :Location (Transformation|GetActorLocation (Game|GetInstigator)) :MaxAudibleDistance 2.5e+03 :IndicatorType "Gunshot" :Team (Team|GetActorTeam (Game|GetInstigator)) :Affiliations (Utilities|Array|MakeArray "Hostile" "Neutral") :Instigator (Game|GetInstigator) :Tint "(R=0.000000,G=0.000000,B=0.000000,A=0.000000)" :AttenuationMaxRangeMultiplier 1.0 :AttenuationIntensityMultiplier 1.0))

(event EventOnStoppedOnServer
  ;[1 1 1] EVENT -> OnStoppedOnServer
  (if (not (Variables|Default|GetExploded))
    (Variables|Default|SetExploded true)
    (Effects|SpawnEmitteratLocation (Variables|Default|GetPS_ProjectileBreak) (Transformation|GetActorLocation) (Transformation|GetActorRotation))
    (Collision|SetCollisionEnabled (Variables|Collision|GetCapsuleComponent) "NoCollision")
    (Components|Activation|Deactivate (Variables|Default|GetNiagaraParticleSystemComponent))
    (Delivery|DoExplosion)))

(event Collision|EventHit
  (params (MyComp :type Primitive%20Component%20Object%20Reference) (Other :type Actor%20Object%20Reference) (OtherComp :type Primitive%20Component%20Object%20Reference) (bSelfMoved :type Boolean) (HitLocation :type Vector :default "0, 0, 0") (HitNormal :type Vector :default "0, 0, 0") (NormalImpulse :type Vector :default "0, 0, 0") (Hit :type Hit%20Result%20Structure%20%28by%20ref%29))
  ;[1 1 1] EVENT -> Hit
  (if (not (Variables|Default|GetExploded))
    (Variables|Default|SetExploded true)
    (Effects|SpawnEmitteratLocation (Variables|Default|GetPS_ProjectileBreak) HitLocation (Math|Rotator|MakeRotfromX HitNormal))
    (Game|Components|ProjectileMovement|StopSimulating (Variables|Movement|GetProjectileMovementComponent) Hit)
    (Collision|SetCollisionEnabled (Variables|Collision|GetCapsuleComponent) "NoCollision")
    (Components|Activation|Deactivate (Variables|Default|GetNiagaraParticleSystemComponent))
    (Delivery|DoExplosion)))

(event EventOnExploded
  (params (HitActors :type Array%20of%20Actor%20Object%20References) (HitResults :type Array%20of%20Hit%20Result%20Structures))
  ;[1 1 1] EVENT -> OnExploded
  (AI|MakeNoiseAtAIProjectile
    (:Then
      (Audio|SpawnSoundAttached (Variables|Default|GetProjectielDirtImpact_Cue) (Variables|ProjectileDeimosRanged|GetRanged_Projectile_Sound))
      (bind _ranged_projectile_sound (Variables|ProjectileDeimosRanged|GetRanged_Projectile_Sound)
        (:then
          (Audio|Components|Audio|FadeOut _ranged_projectile_sound 0.2 0.0 "Linear" :!class /Script/Engine.AudioComponent))
        (:else)))))

(event EventDestroyed
  ;[1 1 1] EVENT -> OnExploded
  (bind _ranged_projectile_sound (Variables|ProjectileDeimosRanged|GetRanged_Projectile_Sound)
    (:then
      (if (Audio|Components|Audio|IsPlaying _ranged_projectile_sound)
        (Audio|Components|Audio|FadeOut _ranged_projectile_sound 0.2 0.0 "Linear" :!class /Script/Engine.AudioComponent)))
    (:else)))

(fn ConstructionScript)
